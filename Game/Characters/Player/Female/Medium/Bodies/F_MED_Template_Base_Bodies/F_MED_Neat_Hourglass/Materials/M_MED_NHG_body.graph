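# Body material for the medium female player character's Neat Hourglass template base body, using Default Lit shading.
# It holds a normal map texture sample and two scalar constants, with no connections to material outputs shown.
# It is flagged for use on skeletal meshes with morph targets and mesh deformers.
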Material /Game/Characters/Player/Female/Medium/Bodies/F_MED_Template_Base_Bodies/F_MED_Neat_Hourglass/Materials/M_MED_NHG_body
graph {
 properties {
  bUsedWithMeshDeformer=true
  bUsedWithMorphTargets=true
  bUsedWithSkeletalMesh=true
  pixelDepthOffsetMode=PDOM_Legacy
  usedShadingModels=Default Lit
 }
 nodes [
  Constant() {r=0.2}
  Constant() {r=0.5}
  TextureSample() {
   samplerType=SAMPLERTYPE_Normal
   texture=&/MED_NHG_normal
  }
 ]
}
